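; Geode god ray effect Blueprint that fades a light ray mesh in and out.
; The replicated IsRayVisible state starts at 0 on BeginPlay, and its OnRep calls the matching fade event.
; Fades are material-driven: fade-in unhides the ray and writes game-time stamps, pushing fade-out far into the future.
; Fade-out stamps current time plus FadeOutDuration, then hides the ray after a fixed delay, skipping if already hidden.
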(event EventBeginPlay
  (Variables|Default|SetIsRayVisible 0))

(event Custom|Ray_FadeIn
  (bind _returnvalue (Utilities|Time|GetGameTimeinSeconds))
  (bind _ray (Variables|Default|GetRay))
  (Development|SetHiddeninGame _ray false false)
  (Rendering|Material|SetScalarParameterValueonMaterials _ray "FadeInTime" _returnvalue)
  (Rendering|Material|SetScalarParameterValueonMaterials _ray "FadeOutTIme" (+ _returnvalue 1e+04)))

(event Custom|Ray_Fadeout
  (bind _ray (Variables|Default|GetRay))
  (if (not (Variables|/Script/Engine.SceneComponent|GetbHiddenInGame (Variables|Default|GetRay)))
    (Rendering|Material|SetScalarParameterValueonMaterials _ray "FadeOutTIme" (+ (Utilities|Time|GetGameTimeinSeconds) (Variables|Default|GetFadeOutDuration)))
    (Utilities|FlowControl|Delay 5.0)
    (Development|SetHiddeninGame _ray true false)))

(fn OnRep_IsRayVisible
  (switch int (Variables|Default|GetIsRayVisible)
    (:0
      (CallFunction|RayFadeout))
    (:1
      (CallFunction|RayFadeIn))))

(fn ConstructionScript)
